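# Physics asset for the HawtChamp backpack cosmetic, giving its dangling character rig secondary motion through capsule bodies on dyn_ bones.
# Constrained chains branch from dyn_spine_1 into the upper spine, arms and legs, with a peel chain on dyn_spine_3 and a tag on the left leg.
# Arms swing widest, the left leg is stiffer than the right, and twist is mostly zero.
# The spine_2 to spine_1 constraint appears twice with different limits.
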PhysicsAsset /BRCosmetics/Accessories/FORT_Backpacks/Backpack_M_MED_HawtChamp/Meshes/M_MED_HawtChamp_Pack_Physics
bodies {
 dyn_hand_l_1 [{shapeType=Capsule, length=2.0627, radius=0.75}]
 dyn_hand_l_2 [{shapeType=Capsule, length=1.4197, radius=1.0749}]
 dyn_hand_l_3 [{shapeType=Capsule, length=1, radius=1.1641}]
 dyn_hand_r_1 [{shapeType=Capsule, length=1.4513, radius=1.113}]
 dyn_hand_r_2 [{shapeType=Capsule, length=1.0299, radius=1.2177}]
 dyn_hand_r_3 [{shapeType=Capsule, length=0.8882, radius=1.3276}]
 dyn_leg_l_1 [{shapeType=Capsule, length=1, radius=1.057}]
 dyn_leg_l_2 [{shapeType=Capsule, length=1.6421, radius=1.067}]
 dyn_leg_l_3 [{shapeType=Capsule, length=0.6839, radius=2.143, center=<0.366 0 0>}]
 dyn_leg_l_4 [{shapeType=Capsule, length=2, radius=0.545, center=<-0.5028 0 0>}]
 dyn_leg_r_1 [{shapeType=Capsule, length=1.0469, radius=1.0571}]
 dyn_leg_r_2 [{shapeType=Capsule, length=1.75, radius=1.0674, center=<0.1513 0 0>}]
 dyn_leg_r_3 [{shapeType=Capsule, length=0.0057, radius=2.1434, center=<-0.2663 0.0712 0>}]
 dyn_leg_r_4 [{shapeType=Capsule, length=2.2903, radius=0.5459, center=<0.6175 0 0>}]
 dyn_peel_1 [{shapeType=Capsule, length=3.1067, radius=0.75, center=<-0.0114 -0.2943 -0.0071>, rotation=<rot 0 177.7765 -88.6149>}]
 dyn_peel_2 [{shapeType=Capsule, length=1.8987, radius=0.8934, rotation=<rot -3.9287 -90 4.1129>}]
 dyn_peel_3 [{shapeType=Capsule, length=1.2591, radius=0.9728, center=<0.2279 0.0034 -0.0292>, rotation=<rot 0.4623 -90 -4.5376>}]
 dyn_spine_1 [{shapeType=Capsule, length=8, radius=3.2192, center=<1.5575 0 0>}]
 dyn_spine_2 [{shapeType=Capsule, length=3.7962, radius=4.2866, center=<2.9287 0 0>}]
 dyn_spine_3 [{
   shapeType=Capsule
   length=0.7737
   radius=2.9041
   shapeName=dyn_spine_3_capsule
   center=<3.3072 0.2486 -0.1741>
   rotation=<rot 0 -0.0002 34.9999>
  }]
 dyn_tag [{shapeType=Capsule, length=0.255, radius=0.755, center=<1.1286 0 0>, rotation=<rot -0.0001 0.0003 -90.0004>}]
}
constraints [
 {bone1Name=dyn_spine_2, bone2Name=dyn_spine_1, swing1LimitDegrees=20, swing2LimitDegrees=20, twistLimitDegrees=0}
 {bone1Name=dyn_peel_1, bone2Name=dyn_spine_3, swing1LimitDegrees=10, swing2LimitDegrees=10, twistLimitDegrees=0}
 {bone1Name=dyn_peel_2, bone2Name=dyn_peel_1, swing1LimitDegrees=20, swing2LimitDegrees=20, twistLimitDegrees=3}
 {bone1Name=dyn_peel_3, bone2Name=dyn_peel_2, swing1LimitDegrees=30, swing2LimitDegrees=30, twistLimitDegrees=5}
 {bone1Name=dyn_leg_r_1, bone2Name=dyn_spine_1, swing1LimitDegrees=20, swing2LimitDegrees=50, twistLimitDegrees=0}
 {bone1Name=dyn_leg_r_2, bone2Name=dyn_leg_r_1, swing1LimitDegrees=20, swing2LimitDegrees=40, twistLimitDegrees=0}
 {bone1Name=dyn_leg_r_3, bone2Name=dyn_leg_r_2, swing1LimitDegrees=30, swing2LimitDegrees=50, twistLimitDegrees=0}
 {bone1Name=dyn_leg_r_4, bone2Name=dyn_leg_r_3, swing1LimitDegrees=40, swing2LimitDegrees=30, twistLimitDegrees=0}
 {bone1Name=dyn_leg_l_1, bone2Name=dyn_spine_1, swing1LimitDegrees=0, swing2LimitDegrees=5, twistLimitDegrees=0}
 {bone1Name=dyn_leg_l_2, bone2Name=dyn_leg_l_1, swing1LimitDegrees=0, swing2LimitDegrees=20, twistLimitDegrees=0}
 {bone1Name=dyn_leg_l_3, bone2Name=dyn_leg_l_2, swing1LimitDegrees=0, swing2LimitDegrees=20, twistLimitDegrees=0}
 {bone1Name=dyn_leg_l_4, bone2Name=dyn_leg_l_3, swing1LimitDegrees=10, swing2LimitDegrees=30, twistLimitDegrees=0}
 {bone1Name=dyn_tag, bone2Name=dyn_leg_l_2, swing1LimitDegrees=5, swing2LimitDegrees=45, twistLimitDegrees=10}
 {bone1Name=dyn_hand_r_1, bone2Name=dyn_spine_1, swing1LimitDegrees=70, swing2LimitDegrees=70, twistLimitDegrees=0}
 {bone1Name=dyn_hand_r_2, bone2Name=dyn_hand_r_1, swing1LimitDegrees=70, swing2LimitDegrees=70, twistLimitDegrees=0}
 {bone1Name=dyn_hand_r_3, bone2Name=dyn_hand_r_2, swing1LimitDegrees=70, swing2LimitDegrees=70, twistLimitDegrees=0}
 {bone1Name=dyn_hand_l_1, bone2Name=dyn_spine_1, swing1LimitDegrees=70, swing2LimitDegrees=70, twistLimitDegrees=0}
 {bone1Name=dyn_hand_l_2, bone2Name=dyn_hand_l_1, swing1LimitDegrees=70, swing2LimitDegrees=70, twistLimitDegrees=0}
 {bone1Name=dyn_hand_l_3, bone2Name=dyn_hand_l_2, swing1LimitDegrees=70, swing2LimitDegrees=70, twistLimitDegrees=0}
 {bone1Name=dyn_spine_3, bone2Name=dyn_spine_2, swing1LimitDegrees=5, swing2LimitDegrees=14, twistLimitDegrees=0}
 {bone1Name=dyn_spine_2, bone2Name=dyn_spine_1, swing1LimitDegrees=0, swing2LimitDegrees=12, twistLimitDegrees=0}
]
